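# Translucent, unlit, two-sided mesh-particle material for the DudeBro enemy's sweep beam effect.
# A dynamic parameter drives SplineThicken width, while panning, distorted smoke textures and a sine edge profile feed a three-color ramp, boosted to HDR with particle color.
# Opacity combines funnel-shaped UV masks and smoke, subtracts dynamic-parameter erosion, and switches between offset-driven wipe and depth fade.
# Optional sphere mask and particle alpha further shape opacity.
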Material /Game/Characters/Enemies/DUDEBRO/FX/Sweep/M_DudeBro_Sweep_Beam
graph {
 properties {
  bUsedWithMeshParticles=true
  bUsedWithNiagaraMeshParticles=true
  blendMode=BLEND_Translucent
  pixelDepthOffsetMode=PDOM_Legacy
  shadingModel=MSM_Unlit
  twoSided=true
  usedShadingModels=Unlit
 }
 nodes [
  n1 = LinearInterpolate(
   B=ScalarParameter() {defaultValue=250, parameterName=Width}
   Alpha=DynamicParameter() {defaultValue=<rgba 1 0 0 8>, paramNames [Width, Erode, ErodeMaskUVOffset, ErodeNoise]}.Width
  )
  MaterialFunctionCall(WidthBase=@n1, WidthTip=@n1) {materialFunction=&SplineThicken}
  n2 = VectorParameter() {parameterName=ColorLow, defaultValue=<rgba 0.0062 0.0004 0.0208 1>}
  n3 = VectorParameter() {parameterName=ColorHigh, defaultValue=<rgba 1 0.2854 0.9202 1>}
  n4 = MaterialFunctionCall(Float2=TextureCoordinate()) {materialFunction=&BreakOutFloat2Components}
  n5 = Sine(None=Multiply(A=@n4.R) {constB=0.5})
  n6 = Power(Base=@n5) {constExponent=5}
  n7 = TextureCoordinate() {vTiling=0.4}
  n8 = TextureSample(UVs=TextureCoordinate() {vTiling=0.6}) {samplerType=SAMPLERTYPE_Normal, texture=&T_Weap_Distort_N_02}
  n9 = Add(
   A=MaterialFunctionCall(XY=Panner(Coordinate=@n7) {speedY=-1.8}) {materialFunction=&Swizzle}.YX
   B=Divide(A=ComponentMask(None=@n8.RGB) {g=true, r=true}) {constB=8}
  )
  n10 = TextureSample(UVs=@n9) {texture=/Game/Effects/Fort_Effects/Textures/Smoke/T_Smoke_Trail_Soft_Chromatic}
  n11 = Multiply(A=@n6, B=@n10.R)
  n12 = MaterialFunctionCall(
   A=@n2.RGB
   B=VectorParameter() {parameterName=ColorMid, defaultValue=<rgba 0.2135 0.0465 0.3281 1>}.RGB
   C=@n3.RGB
   Alpha=Multiply(A=@n11, B=@n11)
  ) {materialFunction=&3ColorBlend}
  n13 = LinearInterpolate(A=@n2.RGB, B=@n3.RGB, Alpha=Add(A=@n11, B=Power(Base=@n5) {constExponent=60}))
  n14 = TextureCoordinate() {vTiling=0.1}
  n15 = TextureSample(UVs=MaterialFunctionCall(XY=Panner(Coordinate=@n14) {speedY=-2}) {materialFunction=&Swizzle}.YX) {samplerType=SAMPLERTYPE_Grayscale, texture=&T_Smoke_Trail_Soft}
  n16 = Panner(Coordinate=TextureCoordinate() {uTiling=0.5}) {speedX=0.5, speedY=-3}
  n17 = Multiply(
   A=@n12.Result
   B=Add(
    A=TextureSample(UVs=@n16) {texture=&T_Inky_Smoke_Tile}.R
    B=Power(Base=@n15.R) {constExponent=0.5}
   )
  )
  n18 = ParticleColor()
  n19 = MaterialFunctionCall(
   LDR Input=Clamp(None=LinearInterpolate(A=@n12.Result, B=@n13) {constAlpha=0.3})
   HDR Dynamic Range=Constant() {r=5}
   HDR Tint=Multiply(A=Multiply(A=@n17, B=@n18.RGB)) {constB=20}
  ) {materialFunction=&DeriveHDRfromLDR}
  MaterialFunctionCall(Emissive=@n19.RemappedRange) {materialFunction=&MF_FX_EmissiveAdjustments}
  n20 = TextureCoordinate() {vTiling=2}
  n21 = OneMinus(None=TextureSample(UVs=@n20) {texture=/Game/Effects/Fort_Effects/Textures/T_VFX_PlasmaClouds}.G)
  n22 = DynamicParameter() {defaultValue=<rgba 1 0 0 8>, paramNames [Width, Erode, ErodeMaskUVOffset, ErodeNoise]}
  Multiply(A=@n21, B=Add(A=ComponentMask(None=@n20) {g=true}, B=@n22.ErodeMaskUVOffset))
  n23 = ParticleColor()
  n24 = ComponentMask(None=TextureCoordinate()) {r=true}
  n25 = Power(Base=Sine(None=Divide(A=@n24))) {constExponent=5}
  n26 = Constant()
  n27 = Subtract(A=OneMinus(None=@n4.G)) {constB=0.93}
  n28 = Power(
   Base=Subtract(A=@n6, B=StaticSwitchParameter(True=@n26, False=@n27) {parameterName=NoTriangleShape?})
   Exp=ScalarParameter() {parameterName=FunnelMask}
  )
  n29 = Multiply(A=Sine(None=Multiply(A=@n4.G) {constB=0.5})) {constB=5}
  n30 = Multiply(
   A=StaticSwitchParameter(True=@n29, False=Saturate(None=@n29)) {parameterName=NoSaturate?}
   B=Add(A=TextureSample(UVs=@n9) {texture=/Game/Effects/Fort_Effects/Textures/Smoke/T_Smoke_Trail_Soft_2}.R, B=@n10.R)
  )
  n31 = ScalarParameter() {defaultValue=3, parameterName=OpacityIntensity}
  n32 = Subtract(
   A=Multiply(A=Multiply(A=@n28, B=@n30), B=@n31) {constB=3}
   B=Multiply(A=@n31, B=DynamicParameter() {defaultValue=<rgba 1 0 0 8>, paramNames [Width, Erode, ErodeMaskUVOffset, ErodeNoise]}.Erode)
  )
  n33 = ComponentMask(None=TextureCoordinate()) {g=true}
  n34 = Add(A=@n33, B=@n22.ErodeMaskUVOffset) {constB=-1}
  n35 = OneMinus(None=LinearInterpolate(Alpha=Clamp(None=@n34) {minDefault=-1}) {constA=-1})
  n36 = Subtract(A=Multiply(A=Clamp(None=@n25), B=Clamp(None=@n32)), B=@n35)
  n37 = DepthFade(Opacity=@n32, FadeDistance=ScalarParameter() {defaultValue=100, parameterName=FadeDistance})
  n38 = StaticSwitchParameter(
   True=Multiply(A=Clamp(None=@n36)) {constB=2}
   False=ShadingPathSwitch(Default=ShadingPathSwitch(Default=@n37, Mobile=@n32) {desc=Migrated from FeatureLevelSwitch}, Forward=@n32, Mobile=@n32)
  ) {parameterName=OpacitySwitchErode}
  n39 = VectorParameter() {parameterName=Sphere Mask Offset, defaultValue=<rgba 0.5 0 0 1>}
  n40 = Power(
   Base=SphereMask(
    A=TextureCoordinate() {vTiling=0.75}
    B=ComponentMask(None=@n39.RGB) {g=true, r=true}
    Radius=ScalarParameter() {defaultValue=0.9, group=Sphere Mask, parameterName=Sphere Mask Radius}
    Hardness=ScalarParameter() {group=Sphere Mask, parameterName=Sphere Mask Hardness}
   )
   Exp=ScalarParameter() {defaultValue=2.5, group=Sphere Mask, parameterName=Sphere Mask Exp}
  ) {constExponent=2.5}
  Multiply(
   A=If(
    A=ScalarParameter() {group=Opacity, parameterName=Use Particle Alpha to Drive opacity}
    "A > B"=@n23.A
    "A == B"=@n23.A
    "A < B"=Constant() {r=1}
   ) {constB=0.99}
   B=LinearInterpolate(A=@n38, B=Multiply(A=@n40, B=@n38), Alpha=ScalarParameter() {group=Sphere Mask, parameterName=Use Sphere Mask})
  )
  n41 = Power(Base=Clamp(None=@n21), Exp=@n22.ErodeNoise)
  OneMinus(None=Clamp(None=Multiply(A=@n41, B=@n35)))
  OneMinus(None=@n33)
  Panner(Coordinate=@n20) {speedY=2}
  Reroute(None=@n19.RemappedRange)
 ]
}
